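# Physics asset for the Mint Snow Globe backpack cosmetic, defining the bodies and joints for a small jointed tree on the pack.
# A capsule on the pack bone anchors a chain of three smaller capsules on the dyn_tree joint bones.
# Each joint is constrained to the one before it with swing and twist limits of a few degrees, so the tree can only wobble slightly.
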PhysicsAsset /BRCosmetics/Accessories/FORT_Backpacks/Backpack_M_MED_Snow_Globe_Mint/Meshes/M_MED_Snow_Globe_Mint_Pack_Physics
bodies {
 dyn_tree_jnt_1 [{shapeType=Capsule, length=1, radius=0.6, center=<0.2687 0 0.0497>, rotation=<rot 79.5129 -180 0>}]
 dyn_tree_jnt_2 [{shapeType=Capsule, length=1, radius=0.6, center=<0.2687 0 0.0497>, rotation=<rot 79.5132 -180 0>}]
 dyn_tree_jnt_3 [{shapeType=Capsule, length=1, radius=0.6, center=<0.2687 0 0.0497>, rotation=<rot 86.4604 -180 0>}]
 pack [{shapeType=Capsule, length=1, radius=2}]
}
constraints [
 {bone1Name=dyn_tree_jnt_1, bone2Name=pack, swing1LimitDegrees=1, swing2LimitDegrees=1, twistLimitDegrees=1}
 {bone1Name=dyn_tree_jnt_2, bone2Name=dyn_tree_jnt_1, swing1LimitDegrees=4, swing2LimitDegrees=4, twistLimitDegrees=2}
 {bone1Name=dyn_tree_jnt_3, bone2Name=dyn_tree_jnt_2, swing1LimitDegrees=2, swing2LimitDegrees=2, twistLimitDegrees=3}
]
