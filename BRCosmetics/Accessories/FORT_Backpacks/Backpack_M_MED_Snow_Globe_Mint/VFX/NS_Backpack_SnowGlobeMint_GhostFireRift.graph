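# Niagara system for the Mint Snow Globe backpack cosmetic, layering a ghost-fire rift with floating shards, cold smoke, snow and a glow.
# Rift and sine-bobbing shard meshes, rising green flipbook flames and the glow follow sockets or bones on User.SkeletalMesh.
# Glow alpha pulses by sine, and its attached light takes User.EyeglowColor.
# Smoke and snow spawn in offset spheres; curl noise and strengthening point attraction swirl the falling snow.
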NiagaraSystem /BRCosmetics/Accessories/FORT_Backpacks/Backpack_M_MED_Snow_Globe_Mint/VFX/NS_Backpack_SnowGlobeMint_GhostFireRift
dependencies {
 usedDataInterfaces [&NiagaraDataInterfaceVectorCurve, &NiagaraDataInterfaceColorCurve, &NiagaraDataInterfaceCurve]
 usedDynamicInputs []
 usedModules [&SystemState]
}
emitters {
 AirGlow {
  emitterUpdateScript [
   &EmitterState()
   &SpawnBurst_Instantaneous(Spawn Count=1, Use Spawn Probability=false)
  ]
  particleSpawnScript [/Niagara/Modules/Spawn/Initialization/InitializeParticle(
    Color=<rgba 0.5972 0.956 0.1683 0.4>
    Module.Write Color=true
    Module.Write Lifetime=true
    Module.Write Mass=true
    Module.Write Material Random=true
    Module.Write Position=true
    Module.Write Scale=false
    Module.Write SpriteRotation=false
    Module.Write SpriteSize=true
    Sprite Size=&Vector2DFromFloat(Value=15)
   )]
  particleUpdateScript [
   &ParticleState()
   &SkeletalMeshLocation(
    Bone / Socket Index=4
    Inherit Velocity=false
    Interpolate Mesh Sampling=false
    MaskSpawnGroup=false
    Mesh Sampling Type=::NewEnumerator5
    Sampled Position Offset Coordinate Space:ENiagaraMeshTransforms=::NewEnumerator3
    Skeletal Mesh=$User.SkeletalMesh
    Socket Sampling Mode=::NewEnumerator5
   )
   &FN_Light_v3(
    Color=$User.EyeglowColor
    Create Color Attribute=true
    Create Exponent Attribute=true
    Create Position Attribute=true
    Create Radius Attribute=true
    Exponent=2
    Position:Vector3f=&Add_Vector(
     A=Linked() {linkedVariable {name=Particles.Position, type=NiagaraPosition}}
     B=&RotateVectorByQuaternion(Vector To Rotate=<0 5 0>, Quaternion=$Output.SkeletalMeshLocation.SampledRotation)
    )
    Radius=10
   )
   &ScaleColor(Scale Alpha=&Sine(Normalized Angle=$Emitter.Age, Period=2, Scale=0.5, Bias=2), ScaleA=true, ScaleRGB=true)
  ]
  rendererClasses [&NiagaraSpriteRendererProperties, &NiagaraLightRendererProperties]
  renderers [
   {rendererClass=&NiagaraSpriteRendererProperties}
   {rendererClass=&NiagaraLightRendererProperties, rendererIndex=1}
  ]
 }
 ColdSmoke {
  emitterUpdateScript [
   &EmitterState(Life Cycle Mode=::NewEnumerator1, Loop Duration=1, UseLoopDelay=true)
   &SpawnRate(SpawnRate=3, Use Spawn Probability=false)
  ]
  particleSpawnScript [
   &InitializeParticle(
    Lifetime=1.5
    Module.Write Color=true
    Module.Write Lifetime=true
    Module.Write Mass=true
    Module.Write Material Random=true
    Module.Write Position=false
    Module.Write Scale=false
    Module.Write SpriteRotation=true
    Module.Write SpriteSize=true
    Position=<0 0 2>
    Sprite Rotation=&RandomRangeFloat(Minimum=0, Maximum=1, Evaluation Type=::NewEnumerator1)
    Sprite Size=&Vector2DFromFloat(Value=&RandomRangeFloat(Minimum=16, Maximum=22, Evaluation Type=::NewEnumerator1))
   )
   &Color(Color=&MakeLinearColorFromVectorAndFloat("Vector (RGB)"=<0.9906 1.1742 1.25>, "Float (Alpha)"=&ClampFloat(Float=0.22)))
   /Niagara/Modules/Spawn/Location/SphereLocation(MaskSpawnGroup=false, Module.Override Randomness=false, Offset=<-5 0 10>, Sphere Radius=5, Surface Only=false)
  ]
  particleUpdateScript [
   &ParticleState()
   &ScaleColor(
    Scale Alpha=&FloatFromCurve(FloatCurve=DataInterface() {propertyValues {Curve {keys [{arriveTangent=1, time=0, value=0}, {arriveTangent=-1, leaveTangent=-1, time=0.5, value=1}, <at 1 0>]}}})
    Scale RGB=&VectorFromCurve(VectorCurve=DataInterface() {propertyValues {XCurve {keys [<at 0 0.3518>,<at 1 0.6903>]}, YCurve {keys [<at 0 0.5139>,<at 1 0.7818>]}, ZCurve {keys [<at 0 1>,<at 1 1>]}}})
    ScaleA=true
    ScaleRGB=true
   )
   /Niagara/Modules/Update/SubUV/SubUVAnimation(Number Of Frames=64)
   &ScaleSpriteSize(Scale Factor=&Vector2DFromFloat(Value=&FloatFromCurve(FloatCurve=DataInterface() {propertyValues {Curve {keys [{arriveTangent=1.8504, interpMode=RCIM_Cubic, time=0, value=0.5}, {interpMode=RCIM_Cubic, leaveTangent=0.4544, time=1, value=1}]}}})))
   &CameraOffset(Camera Offset Amount=10)
   &AccelerationForce(Acceleration=<0 0 -3>)
   &SolveForcesAndVelocity(Clamp Velocity=false, Limit Acceleration=false)
  ]
  rendererClasses [&NiagaraSpriteRendererProperties]
  renderers [{rendererClass=&NiagaraSpriteRendererProperties}]
 }
 Fire_Back {
  emitterUpdateScript [
   &EmitterState()
   &SpawnRate(SpawnRate=&RandomRangeFloat(Minimum=5, Maximum=8), Use Spawn Probability=false)
  ]
  particleSpawnScript [
   &InitializeParticle(
    Lifetime=&UniformRangedFloat(Minimum=1.2, Maximum=1.4)
    Module.Write Color=true
    Module.Write Lifetime=true
    Module.Write Mass=false
    Module.Write Material Random=true
    Module.Write Position=false
    Module.Write Scale=false
    Module.Write SpriteRotation=true
    Module.Write SpriteSize=true
    Position=&SimulationPosition()
    Sprite Rotation=&UniformRangedFloat(Minimum=-180, Maximum=180)
    Sprite Size=&Vector2DFromFloat(Value=&RandomRangeFloat(Minimum=9, Maximum=9))
   )
   &DynamicMaterialParameters(
    Index 0 Param 2=&UniformRangedFloat(Minimum=0, Maximum=1)
    Index 0 Param 3=1
    Index 0 Param 4=&UniformRangedFloat(Minimum=2, Maximum=3)
    Module.Param0WriteEnabled=false
    Module.Param1WriteEnabled=true
    Module.Param2WriteEnabled=false
    Module.Param3WriteEnabled=true
   ) {module=DynamicMaterialParameters002}
   &CameraOffset(Camera Offset Amount=&RandomRangeFloat(Minimum=0, Maximum=0))
   /SetVariables(Particles.Range:Vector3f=&RandomRangeVector(Minimum=<0 -4.3 -1>, Maximum=<0 3 2.2>)) {isSetParameters=true}
  ]
  particleUpdateScript [
   &ParticleState()
   &SkeletalMeshLocation(
    Bone / Socket Index=4
    Inherit Velocity=false
    MaskSpawnGroup=false
    Mesh Sampling Type=::NewEnumerator5
    Sampled Position Offset=&Lerp_Vector(A=$Particles.Range, B=&Add_Vector(A=$Particles.Range, B=<0 0 2>), Alpha=$Particles.NormalizedAge)
    Sampled Position Offset Coordinate Space:ENiagaraMeshTransforms=::NewEnumerator3
    Skeletal Mesh=$User.SkeletalMesh
    Socket Sampling Mode=::NewEnumerator5
   )
   &ScaleColor(
    Scale Mode=::NewEnumerator1
    Scale RGBA=&Vector4FromCurve(Vector4Curve=DataInterface() {propertyValues {
       AlphaCurve {keys [<at 0.0024 0>,<at 0.1271 1>,<at 0.9977 -0.0026>]}
       BlueCurve {keys [<at 0.3937 0>,<at 0.6697 0>]}
       GreenCurve {keys [<at 0.3937 10>,<at 0.6697 1>]}
       RedCurve {keys [<at 0.3937 0.5287>,<at 0.6697 0.2163>]}
      }})
    ScaleRGBA=true
   ) {module=ScaleColor001}
   &SubUVAnimation(Number Of Frames=32)
   &ScaleSpriteSize(Scale Factor=&Vector2DFromFloat(Value=&FloatFromCurve(FloatCurve=DataInterface() {propertyValues {Curve {keys [{arriveTangent=0.9771, interpMode=RCIM_Cubic, leaveTangent=0.9771, tangentMode=RCTM_User, time=0, value=0.7}, <at 0.9503 1.2 RCIM_Cubic>]}}})))
   &DynamicMaterialParameters(
    Index 0 Param 1=&FloatFromCurve(FloatCurve=DataInterface() {propertyValues {Curve {keys [{arriveTangent=2.7551, interpMode=RCIM_Cubic, time=0.25, value=1}, <at 0.419 0.898 RCIM_Cubic>, <at 0.7 1>]}}})
    Index 0 Param 2=1
    Index 0 Param 3=&FloatFromCurve(FloatCurve=DataInterface() {propertyValues {Curve {keys [
         <at 0 1.888 RCIM_Cubic>
         {arriveTangent=-0.0385, interpMode=RCIM_Cubic, leaveTangent=-0.0385, tangentMode=RCTM_User, time=0.222, value=0.9757}
         {arriveTangent=1.1537, interpMode=RCIM_Cubic, leaveTangent=1.1537, time=0.9173, value=1.208}
         <at 1.2 2.104 RCIM_Cubic>
        ]}}})
    Index 0 Param 4=1
    Module.Param0WriteEnabled=true
    Module.Param1WriteEnabled=true
    Module.Param2WriteEnabled=true
    Module.Param3WriteEnabled=false
   ) {module=DynamicMaterialParameters001}
  ]
  rendererClasses [&NiagaraSpriteRendererProperties]
  renderers [{rendererClass=&NiagaraSpriteRendererProperties}]
 }
 Rift {
  emitterUpdateScript [
   &EmitterState(
    Inactive Response=::NewEnumerator1
    Life Cycle Mode=::NewEnumerator1
    Loop Behavior=::NewEnumerator1
    Loop Duration Mode=::NewEnumerator1
    UseLoopDelay=false
   )
   &SpawnBurst_Instantaneous(Spawn Count=1, Use Spawn Probability=false) {module=SpawnBurst_Instantaneous001}
  ]
  particleSpawnScript [
   &InitializeParticle(
    Color=<rgba 10 10 10 1>
    Lifetime=10
    Mesh Scale=&VectorFromFloat(Value=0.22)
    Module.Write Color=true
    Module.Write Lifetime=true
    Module.Write Mass=true
    Module.Write Material Random=true
    Module.Write Position=false
    Module.Write Scale=true
    Module.Write SpriteRotation=false
    Module.Write SpriteSize=true
    Position=&SimulationPosition()
    Sprite Size=<5 5>
   )
   &CameraOffset(Camera Offset Amount=&RandomRangeFloat(Minimum=0, Maximum=0))
   &InitialMeshOrientation(Rotation=<0.45 0 0>, Use Orientation Vector=true, Use Rotation Vector=true)
  ]
  particleUpdateScript [
   &ParticleState(Kill Particles When Lifetime Has Elapsed=false, Let Infinitely Lived Particles Die When Emitter Deactivates=true)
   &DynamicMaterialParameters(
    Index 0 Param 1=0.16
    Index 0 Param 2=1
    Index 0 Param 3=&Multiply_Float(A=0.05, B=$Emitter.Age)
    Index 0 Param 4=1
    Module.Param0WriteEnabled=true
    Module.Param1WriteEnabled=true
    Module.Param2WriteEnabled=true
    Module.Param3WriteEnabled=true
   )
   &SkeletalMeshLocation(
    Inherit Velocity=false
    MaskSpawnGroup=false
    Mesh Sampling Type=::NewEnumerator5
    Skeletal Mesh=$User.SkeletalMesh
    Socket Sampling Mode=::NewEnumerator5
   )
  ]
  rendererClasses [&NiagaraMeshRendererProperties]
  renderers [{rendererClass=&NiagaraMeshRendererProperties, rendererIndex=0}]
 }
 Shards {
  emitterUpdateScript [
   &EmitterState(
    Inactive Response=::NewEnumerator1
    Life Cycle Mode=::NewEnumerator1
    Loop Behavior=::NewEnumerator1
    Loop Duration Mode=::NewEnumerator1
    UseLoopDelay=false
   )
   &SpawnBurst_Instantaneous(Spawn Count=2, Use Spawn Probability=false)
  ]
  particleSpawnScript [
   /Niagara/Modules/Spawn/Initialization/V2/InitializeParticle(
    Color=<rgba 1 1 1 0.2>
    Lifetime=5
    Mesh Scale Max=<0.075 0.09 0.075>
    Mesh Scale Min=<0.05 0.09 0.05>
    Mesh Scale Mode=::NewEnumerator2
    Sprite Size Mode=::NewEnumerator3
    Uniform Sprite Size=&RandomRangeFloat(Minimum=10, Maximum=10)
    UsePositionOffset=false
    Write Material Random=true
   )
   /SetVariables() {isSetParameters=true}
   &FN_SetSpawnGroup(Number of Spawn Groups=3)
   &InitialMeshOrientation(
    Orientation Axis=<0 1 0>
    Rotation=&MaskVectorBySpawnGroup(Masked Vector=<0 0.55 0>, Passthrough Vector=<0 0.35 0>)
    Use Orientation Vector=true
    Use Rotation Vector=true
   )
  ]
  particleUpdateScript [
   &ParticleState(Kill Particles When Lifetime Has Elapsed=false, Let Infinitely Lived Particles Die When Emitter Deactivates=true)
   &SkeletalMeshLocation(
    Bone / Socket Index=&MaskIntBySpawnGroup(Masked Int=1, Passthrough Int=2, Spawn Group Mask=0)
    Inherit Velocity=false
    MaskSpawnGroup=false
    Mesh Sampling Type=::NewEnumerator5
    Sampled Position Offset=&MakeVector(
     X=&Sine(Normalized Angle=$Emitter.Age, Period=&RandomRangeFloat(Minimum=3, Maximum=3.3), Scale=0.25, Bias=0)
     Y=&Sine(Normalized Angle=$Emitter.Age, Period=&RandomRangeFloat(Minimum=3, Maximum=3.3), Scale=0.25, Bias=0)
     Z=&Sine(Normalized Angle=$Emitter.Age, Period=&RandomRangeFloat(Minimum=3, Maximum=3.3), Scale=0.25, Bias=0)
    )
    Skeletal Mesh=$User.SkeletalMesh
    Socket Sampling Mode=::NewEnumerator5
   )
   /SetVariables(Particles.Position=&Add_Vector(
     A=Linked() {linkedVariable {name=Particles.Initial.Position, type=NiagaraPosition}}
     B=&Add_Vector(
      A=&MakeVector(
       X=&Sine(Normalized Angle=$Emitter.Age, Period=&RandomRangeFloat(Minimum=4.3, Maximum=4.5), Scale=0.25, Bias=0)
       Y=&Sine(Normalized Angle=$Emitter.Age, Period=&RandomRangeFloat(Minimum=4.3, Maximum=4.5), Scale=0.25, Bias=0)
       Z=&Sine(Normalized Angle=$Emitter.Age, Period=&RandomRangeFloat(Minimum=4.3, Maximum=4.5), Scale=0.25, Bias=0)
      )
      B=&MakeVector(
       X=&Sine(Normalized Angle=$Emitter.Age, Period=&RandomRangeFloat(Minimum=3.1, Maximum=3.3), Scale=0.25, Bias=0)
       Y=&Sine(Normalized Angle=$Emitter.Age, Period=&RandomRangeFloat(Minimum=3.1, Maximum=3.3), Scale=0.25, Bias=0)
       Z=&Sine(Normalized Angle=$Emitter.Age, Period=&RandomRangeFloat(Minimum=3.1, Maximum=3.3), Scale=0.25, Bias=0)
      )
     )
    )) {enabled=false, isSetParameters=true}
   &SkeletalMeshLocation(
    Inherit Velocity=false
    MaskSpawnGroup=false
    Module.Override Randomness=false
    Skeletal Mesh=DataInterface() {propertyValues {
      DefaultMesh=null
      ExcludeBoneName=root
      OptionalComponentClass=null
      PreviewMesh=/BRCosmetics/Accessories/FORT_Backpacks/Backpack_M_MED_Snow_Globe_Mint/Meshes/M_MED_Snow_Globe_Mint_Pack
      SkinningMode=SkinOnTheFly
      SoftSourceActor=null
      SourceMode=Default
      UvSetIndex=0
      WholeMeshLOD=-1
      bExcludeBone=false
      bReadDeformedGeometry=true
      bRequireCurrentFrameData=true
      ComponentTags []
      FilteredBones [root]
      FilteredSockets []
      MeshUserParameter {parameter {name=None, typeDefHandle {registeredTypeIndex=1}, varData []}}
      SamplingRegions []
     }}
   ) {enabled=false, module=SkeletalMeshLocation001}
  ]
  rendererClasses [&NiagaraMeshRendererProperties]
  renderers [{rendererClass=&NiagaraMeshRendererProperties, rendererIndex=0}]
 }
 Snow {
  emitterUpdateScript [
   &EmitterState()
   &SpawnRate(SpawnRate=20, Use Spawn Probability=false)
  ]
  particleSpawnScript [
   &InitializeParticle(
    Color Mode=::NewEnumerator0
    Lifetime Max=4
    Lifetime Min=3
    Lifetime Mode=::NewEnumerator1
    Sprite Size Mode=::NewEnumerator1
    Uniform Sprite Size Max=1.4
    Uniform Sprite Size Min=1
    UsePositionOffset=false
    Write Material Random=true
   )
   /SetVariables(Particles.SpriteRotationRate=&RandomRangeFloat(Minimum=-0.25, Maximum=0.25, Evaluation Type=::NewEnumerator1)) {isSetParameters=true}
   &AddVelocity(Velocity=&RandomRangeVector(Minimum=<-10 -10 -10>, Maximum=<10 10 10>, Evaluation Type=::NewEnumerator1))
   &SphereLocation(Hemisphere Z=true, MaskSpawnGroup=false, Module.Override Randomness=false, Offset=<-5 0 6>, Sphere Radius=9, Surface Only=true)
  ]
  particleUpdateScript [
   &ParticleState()
   &Drag(Drag=10, Use Linear Drag=true, Use Rotational Drag=false)
   &AccelerationForce(Acceleration=<0 0 -25>)
   &PointAttractionForce(
    Attraction Strength=&FloatFromCurve(FloatCurve=DataInterface() {propertyValues {Curve {keys [<at 0 0>,<at 5 15>]}}}, CurveIndex=$Particles.Age)
    Attractor Position Offset:Vector3f=<-5 0 6>
    Kill Within Radius=false
    Module.Use Falloff=true
    UseAttractorOffset=true
   )
   &CurlNoiseForce(Noise Frequency=250, Pan Noise=false)
   &SolveForcesAndVelocity(Clamp Velocity=false, Limit Acceleration=false)
   &SpriteRotationRate(Rotation Rate=$Particles.SpriteRotationRate)
   &ScaleSpriteSize(Scale Factor=&Vector2DFromFloat(Value=&FloatFromCurve(FloatCurve=DataInterface() {propertyValues {Curve {keys [
          {arriveTangent=0.9932, time=0, value=0}
          {arriveTangent=-0.7835, leaveTangent=-0.7835, time=0.2002, value=0.9932}
          {arriveTangent=0.7991, leaveTangent=0.7991, time=0.3821, value=0.2097}
          {arriveTangent=-1.0088, leaveTangent=-1.0088, time=0.5993, value=1.0088}
          <at 1 0>
         ]}}})))
   &Color(Color=&MakeLinearColorFromVectorAndFloat(
     "Vector (RGB)"=&VectorFromCurve(VectorCurve=DataInterface() {propertyValues {XCurve {keys [<at 0 1>,<at 1 1>]}, YCurve {keys [<at 0 1>,<at 1 1>]}, ZCurve {keys [<at 0 1>,<at 1 1>]}}})
     "Float (Alpha)"=&ClampFloat(Float=&FloatFromCurve(FloatCurve=DataInterface() {propertyValues {Curve {keys [{arriveTangent=-1, time=0, value=1}, <at 1 0>]}}}))
    ))
   &ScaleColor(Scale Alpha=0.15, ScaleA=true, ScaleRGB=true)
   &CameraOffset()
  ]
  rendererClasses [&NiagaraSpriteRendererProperties]
  renderers [{rendererClass=&NiagaraSpriteRendererProperties}]
 }
}
summary {
 systemName=NS_Backpack_SnowGlobeMint_GhostFireRift
 emitters [
  {emitterName=Rift, rendererClasses [&NiagaraMeshRendererProperties]}
  {emitterName=Fire_Back, rendererClasses [&NiagaraSpriteRendererProperties]}
  {emitterName=Shards, rendererClasses [&NiagaraMeshRendererProperties]}
  {emitterName=ColdSmoke, rendererClasses [&NiagaraSpriteRendererProperties]}
  {emitterName=Snow, rendererClasses [&NiagaraSpriteRendererProperties]}
  {emitterName=AirGlow, rendererClasses [&NiagaraSpriteRendererProperties, &NiagaraLightRendererProperties]}
 ]
 userVariables [
  User.EyeglowColor {type=LinearColor, description="", defaultValue {struct=LinearColor, value=<rgba 1 1 1 1>}}
  User.SkeletalMesh {
   type=NiagaraDataInterfaceSkeletalMesh
   defaultValue {
    struct=NiagaraExt_VariableValue_DataInterface
    value {dataInterface=&self:NiagaraDataInterfaceSkeletalMesh_0}
   }
  }
 ]
}
